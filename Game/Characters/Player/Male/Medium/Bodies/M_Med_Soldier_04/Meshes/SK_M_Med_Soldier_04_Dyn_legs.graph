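# Physics asset for the leg dynamics of the medium male player body M_Med_Soldier_04, covering its dangling pouch bones.
# Small capsules on the right pouch master, left pocket master and left pouch pocket bones hang from a larger pelvis capsule.
# Both master bones connect to the pelvis with free swing and twist.
# The left pocket bone chains to its master through a constraint with tight angular limits.
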PhysicsAsset /Game/Characters/Player/Male/Medium/Bodies/M_Med_Soldier_04/Meshes/SK_M_Med_Soldier_04_Dyn_legs
bodies {
 dyn_pouch_r_MSTR [{shapeType=Capsule, length=0.691, radius=0.5, rotation=<rot 90 0 0>}]
 dyn_pouchpocket_l [{shapeType=Capsule, length=2, radius=1, center=<0.4413 0 0>, rotation=<rot 90 0 0>}]
 dyn_pouchpocket_l_MSTR [{shapeType=Capsule, length=0.691, radius=0.5, rotation=<rot 90 0 0>}]
 pelvis [{shapeType=Capsule, length=10.1692, radius=9.2472}]
}
constraints [
 {bone1Name=dyn_pouchpocket_l, bone2Name=dyn_pouchpocket_l_MSTR, swing1LimitDegrees=20, swing2LimitDegrees=15, twistLimitDegrees=5}
 {
  bone1Name=dyn_pouchpocket_l_MSTR
  bone2Name=pelvis
  swing1LimitDegrees=45
  swing1Motion=Free
  swing2LimitDegrees=45
  swing2Motion=Free
  twistLimitDegrees=45
  twistMotion=Free
 }
 {
  bone1Name=dyn_pouch_r_MSTR
  bone2Name=pelvis
  swing1LimitDegrees=45
  swing1Motion=Free
  swing2LimitDegrees=45
  swing2Motion=Free
  twistLimitDegrees=45
  twistMotion=Free
 }
]
