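# Animation Blueprint for the M_MED_Bronto player body cosmetic, adding secondary motion to pouches, scarf and armband over the pose copied from its attached parent mesh.
# Weighted constraints drive master bones, then additive offsets tilt the right pouch when jumping, falling or crouch-moving (alpha fed by 2D speed).
# Two RigidBody passes reuse Soldier_04 physics assets with skydiving force input, followed by spring-based AnimDynamics on three armband bones.
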AnimBlueprint /BRCosmetics/Characters/Player/Male/Medium/Bodies/M_MED_Bronto/Meshes/M_MED_Bronto_Skeleton_AnimBlueprint
target_skeleton=&/M_MED_Bronto_Skeleton
graphs [
 AnimGraph {nodes [
   n1 = "Constraint - dyn_pouchpocket_l_MSTR (PP)"!(
    ComponentPose=Local To Component(LocalPose=Copy Pose From Mesh() {bUseAttachedParent=true})
    Alpha=1
    ConstraintWeights_0=0.4758
    ConstraintWeights_1:float=0.5242
   ) -> (Pose) {
    alphaBoolBlend {blendOption=Linear}
    boneToModify {boneName=dyn_pouchpocket_l_MSTR}
    constraintSetup [{targetBone {boneName=dyn_pouchpocket_l}}, {targetBone {boneName=thigh_twist_01_l}}]
    constraintWeights [1,1]
   }
   n2 = "Constraint - dyn_pouch_r_MSTR (PPPP)"!(
    ComponentPose=@n1
    Alpha=1
    ConstraintWeights_0=0.106
    ConstraintWeights_1:float=0.0506
    ConstraintWeights_2:float=0.0348
    ConstraintWeights_3:float=0.8086
   ) -> (Pose) {
    alphaBoolBlend {blendOption=Linear}
    boneToModify {boneName=dyn_pouch_r_MSTR}
    constraintSetup [{targetBone {boneName=spine_01}}, {targetBone {boneName=spine_02}}, {targetBone {boneName=spine_03}}, {targetBone {boneName=pelvis}}]
    constraintWeights [1,1,1,1]
   }
   n3 = "Constraint - dyn_scarf_f_MSTR (PPPPP)"!(
    ComponentPose=@n2
    Alpha=1
    ConstraintWeights_0=0.0668
    ConstraintWeights_1:float=0.0227
    ConstraintWeights_2:float=0.3471
    ConstraintWeights_3:float=0.4728
    ConstraintWeights_4:float=0.0906
   ) -> (Pose) {
    alphaBoolBlend {blendOption=Linear}
    boneToModify {boneName=dyn_scarf_f_MSTR}
    constraintSetup [
     {targetBone {boneName=clavicle_r}}
     {targetBone {boneName=spine_03}}
     {targetBone {boneName=spine_04}}
     {targetBone {boneName=spine_05}}
     {targetBone {boneName=clavicle_l}}
    ]
    constraintWeights [1,1,1,1,1]
   }
   n4 = "Transform (Modify) Bone - Bone: dyn_pouch_r_MSTR"(
    ComponentPose=@n3
    Alpha=Get jump_alpha() -> (jump_alpha:double)
    Translation="-0.023,4.302,-0.525"
    Rotation="15.750,1.660,19.032"
    Scale="1,1,1"
   ) {
    rotationMode=BMM_Additive
    rotationSpace=BCS_BoneSpace
    translationMode=BMM_Additive
    translationSpace=BCS_BoneSpace
    alphaBoolBlend {blendOption=Linear}
    alphaScaleBiasClamp {bInterpResult=true, interpSpeedDecreasing=30, interpSpeedIncreasing=30}
    boneToModify {boneName=dyn_pouch_r_MSTR}
    rotation=<rot 15.75 1.66 19.0321>
    translation=<-0.023 4.302 -0.525>
   }
   n5 = "Transform (Modify) Bone - Bone: dyn_pouch_r_MSTR"(
    ComponentPose=@n4
    Alpha=Get crouching alpha() -> (crouching alpha:double)
    Translation="-0.298,0.240,0.690"
    Rotation="18,0.000,0.000"
    Scale="1,1,1"
   ) {
    rotationMode=BMM_Additive
    rotationSpace=BCS_BoneSpace
    translationMode=BMM_Additive
    translationSpace=BCS_BoneSpace
    alphaBoolBlend {blendOption=Linear}
    alphaScaleBiasClamp {bInterpResult=true, interpSpeedDecreasing=50, interpSpeedIncreasing=50}
    boneToModify {boneName=dyn_pouch_r_MSTR}
    rotation {pitch=18}
    translation=<-0.2981 0.2403 0.69>
   }
   n6 = Break Fort Anim Input Skydiving External Force(FortAnimInput_SkydivingExternalForce=Get SkydivingExternalForceInput()) -> (
    bUseSkydivingVectorForce=~
    SkydivingVectorForce=~
    HeadToPelvisDirection=~
    FloatingMultiplier=~
    FloatingAdditive=~
    DivingMultiplier=~
    DivingAdditive=~
    ParachutingMultiplier=~
    ParachutingAdditive=~
   )
   n7 = RigidBody(ComponentPose=@n5, ExternalForce=@n6.SkydivingVectorForce) {
    bClampLinearTranslationLimitToRefPose=true
    bOverrideWorldGravity=true
    overridePhysicsAsset=/Game/Characters/Player/Male/Medium/Bodies/M_Med_Soldier_04/Meshes/SK_M_Med_Soldier_04_Dyn
    simulationSpace=BaseBoneSpace
    alphaBoolBlend {blendOption=Linear}
    baseBoneRef {boneName=spine_02}
    componentAppliedLinearAccClamp=<3000 3000 3000>
    componentLinearAccScale=<0.3 0.3 0.3>
    componentLinearVelScale=<0.3 0.3 0.3>
    overrideWorldGravity {z=-980}
   }
   n8 = RigidBody(ComponentPose=@n7, ExternalForce=@n6.SkydivingVectorForce) {
    bClampLinearTranslationLimitToRefPose=true
    bOverrideWorldGravity=true
    overridePhysicsAsset=/Game/Characters/Player/Male/Medium/Bodies/M_Med_Soldier_04/Meshes/SK_M_Med_Soldier_04_Dyn_legs
    simulationSpace=BaseBoneSpace
    alphaBoolBlend {blendOption=Linear}
    baseBoneRef {boneName=pelvis}
    componentAppliedLinearAccClamp=<3000 3000 3000>
    componentLinearAccScale=<0.3 0.3 0.3>
    componentLinearVelScale=<0.3 0.3 0.3>
    overrideWorldGravity {z=-980}
   }
   n9 = Anim Dynamics - Bone: dyn_armband_knot(ComponentPose=@n8, Alpha=0.75, ExternalForce="0,0,0") {
    alpha=0.9
    angularDampingOverride=0.98
    angularSpringConstant=1500
    bAngularSpring=true
    bGravityOverrideInSimSpace=true
    bOverrideAngularDamping=true
    numSolverIterationsPostUpdate=4
    numSolverIterationsPreUpdate=16
    simulationSpace=Actor
    alphaBoolBlend {blendOption=Linear}
    boundBone {boneName=dyn_armband_knot}
    physicsBodyDefinitions [{
      sphereCollisionRadius=0
      boundBone {boneName=dyn_armband_knot}
      boxExtents=<15 5 5>
      constraintSetup {angularTargetAxis=AxisZ, twistAxis=AxisY, angularLimitsMax=<15 0 15>, angularLimitsMin=<-15 0 -15>, angularTarget=<0 0 1>}
      localJointOffset=<0 0 0>
     }]
   }
   n10 = Anim Dynamics - Bone: dyn_armband_low(ComponentPose=@n9, Alpha=0.5, ExternalForce="0,0,0") {
    alpha=0.9
    angularDampingOverride=0.98
    angularSpringConstant=1500
    bAngularSpring=true
    bGravityOverrideInSimSpace=true
    bOverrideAngularDamping=true
    numSolverIterationsPostUpdate=4
    numSolverIterationsPreUpdate=16
    simulationSpace=Actor
    alphaBoolBlend {blendOption=Linear}
    boundBone {boneName=dyn_armband_low}
    physicsBodyDefinitions [{
      sphereCollisionRadius=0
      boundBone {boneName=dyn_armband_low}
      boxExtents=<15 5 5>
      constraintSetup {angularTargetAxis=AxisZ, twistAxis=AxisY, angularLimitsMax=<15 15 15>, angularLimitsMin=<-15 -15 -15>, angularTarget=<0 0 1>}
      localJointOffset=<0 0 0>
     }]
   }
   n11 = Anim Dynamics - Bone: dyn_armband_hi(ComponentPose=@n10, Alpha=0.5, ExternalForce="0,0,0") {
    alpha=0.9
    angularDampingOverride=0.98
    angularSpringConstant=1500
    bAngularSpring=true
    bGravityOverrideInSimSpace=true
    bOverrideAngularDamping=true
    numSolverIterationsPostUpdate=4
    numSolverIterationsPreUpdate=16
    simulationSpace=Actor
    alphaBoolBlend {blendOption=Linear}
    boundBone {boneName=dyn_armband_hi}
    physicsBodyDefinitions [{
      sphereCollisionRadius=0
      boundBone {boneName=dyn_armband_hi}
      boxExtents=<15 5 5>
      constraintSetup {angularTargetAxis=AxisZ, twistAxis=AxisY, angularLimitsMax=<15 15 15>, angularLimitsMin=<-15 -15 -15>, angularTarget=<0 0 1>}
      localJointOffset=<0 0 0>
     }]
   }
   Output Pose(Result=Component To Local(ComponentPose=@n11)) {layerGroup=None}
  ]}
 EventGraph {nodes [
   Event BlueprintUpdateAnimation()
   > n1 = Sequence()
   > n2 = Branch(Condition=Get bIsCrouching())
   > Set crouching alpha(crouching alpha:double=Get PawnSpeed2D()) -> (Output_Get:double)
   Set crouching alpha(execute=@n2.else, crouching alpha=0)
   n3 = Branch(execute=@n1.then_1, Condition=Get bIsJumping())
   Set jump_alpha(execute=[@n3.then, @n4.then], jump_alpha:double=1) -> (Output_Get:double)
   n4 = Branch(execute=@n3.else, Condition=Get bIsFalling())
   Set jump_alpha(execute=@n4.else, jump_alpha=0)
   TryGetPawnOwner()
  ]}
]
variables [
 crouching alpha {type=number}
 jump_alpha {type=number}
]
