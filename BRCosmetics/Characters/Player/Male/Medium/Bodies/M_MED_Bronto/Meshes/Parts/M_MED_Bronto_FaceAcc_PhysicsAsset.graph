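# Physics asset for the face accessory mesh of the male medium Bronto player body cosmetic.
# It holds a single capsule collision body on the head bone, offset and rotated to fit the head.
# No constraints are defined, so the asset provides only a simple head shape with no jointed simulation.
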PhysicsAsset /BRCosmetics/Characters/Player/Male/Medium/Bodies/M_MED_Bronto/Meshes/Parts/M_MED_Bronto_FaceAcc_PhysicsAsset
bodies {head [{shapeType=Capsule, length=15.805, radius=13.6478, center=<4.8245 5.1995 0.0002>, rotation=<rot 0 -125.0866 -89.9974>}]}
constraints []
